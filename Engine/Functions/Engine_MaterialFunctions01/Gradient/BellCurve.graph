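# Material function, exposed to the library, that maps a scalar x onto a bell curve (normal distribution).
# Both outputs raise a constant e to -0.5 times a squared input.
# The "0-1" output first multiplies x by Pi, fitting the curve to x from -1 to 1 with values from 0 to 1.
# The "Standard" output uses x directly and divides by the square root of 2 Pi, so the area under the curve equals 1.
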MaterialFunction /Engine/Functions/Engine_MaterialFunctions01/Gradient/BellCurve
graph {
 properties {bExposeToLibrary=true, description=Gives a Bell Curve also known as a Normal Distribution}
 nodes [
  n1 = Constant() {desc=e, r=2.7183}
  n2 = FunctionInput() {bUsePreviewValueAsDefault=true, description=The input to map to the curve, inputName=x, inputType=FunctionInput_Scalar}
  n3 = Multiply(A=@n2, B=MaterialFunctionCall() {materialFunction=&Pi}.Result)
  n4 = Multiply(A=@n3, B=@n3)
  FunctionOutput(None=Power(Base=@n1, Exp=Multiply(A=@n4) {constB=-0.5})) {description=Takes input on x from -1 to 1 and Returns values of 0-1, outputName=0-1}
  n5 = Multiply(A=Multiply(A=@n2, B=@n2)) {constB=-0.5}
  n6 = MaterialFunctionCall(Multiplier=Constant() {r=2}) {materialFunction=&Pi}
  FunctionOutput(None=Divide(A=Power(Base=@n1, Exp=@n5), B=SquareRoot(None=@n6.Result))) {
   description="Takes input on x from -pi to pi and Returns values scaled so the 2d area of the curve equals 1. "
   outputName=Standard
   sortPriority=1
  }
 ]
}
