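# BoxMask-2D is a library material function that draws a 2D box mask with adjustable center, bounds and edge falloff.
# It subtracts half the bounds from the absolute offset between position (world position by default) and center, then clamps negatives to zero.
# The length of that vector gives distance outside the box, which is divided by falloff, clamped and inverted, yielding one inside and a rounded gradient fading outward.
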MaterialFunction /Engine/Functions/Engine_MaterialFunctions01/Gradient/BoxMask-2D
graph {
 properties {bExposeToLibrary=true, description=Draws a box in 2D Space. Controls for size and falloff., libraryCategoriesText [Misc,Utility]}
 nodes [
  n1 = FunctionInput(Preview=WorldPosition().XYZ) {
   bUsePreviewValueAsDefault=true
   description=The Position Coordinates to measure the box. Default is World Position and can be left default in most cases.
   inputName=A
   inputType=FunctionInput_Vector2
  }
  n2 = FunctionInput() {
   bUsePreviewValueAsDefault=true
   description=The Center Point location of the box
   inputName=B
   inputType=FunctionInput_Vector2
   sortPriority=1
  }
  n3 = FunctionInput() {
   bUsePreviewValueAsDefault=true
   description=The bounds of the cube in X Y Z.
   inputName=Bounds
   inputType=FunctionInput_Vector2
   sortPriority=2
   previewValue=<100 100 100 1>
  }
  n4 = Subtract(A=Abs(None=Subtract(A=@n1, B=@n2)), B=Multiply(A=@n3) {constB=0.5})
  n5 = FunctionInput() {
   bUsePreviewValueAsDefault=true
   description=Falloff of the edge gradient in units.
   inputName=Edge Falloff
   inputType=FunctionInput_Scalar
   sortPriority=3
   previewValue=<1 0 0 1>
  }
  n6 = Divide(
   A=MaterialFunctionCall(Vector 2=Max(A=@n4) {constB=0}) {materialFunction=/Engine/Functions/Engine_MaterialFunctions02/Utility/VectorLength}.V2 Length
   B=@n5
  )
  FunctionOutput(None=OneMinus(None=Clamp(None=@n6)))
 ]
}
