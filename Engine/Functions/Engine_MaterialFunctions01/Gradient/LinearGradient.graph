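# LinearGradient is a library material function that outputs linear gradients along the U and V axes of a UV set.
# Its Vector2 UV Channel input defaults to texture coordinate 0 when left unconnected.
# Component masks pass the red component to UGradient and the green component to VGradient, with no remapping or other math.
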MaterialFunction /Engine/Functions/Engine_MaterialFunctions01/Gradient/LinearGradient
graph {
 properties {
  bExposeToLibrary=true
  description=Uses UV Channel 0 to generate a linear gradient along the U or V axis
  libraryCategoriesText [Misc,Gradient]
 }
 nodes [
  n1 = FunctionInput(Preview=TextureCoordinate()) {bUsePreviewValueAsDefault=true, inputName=UV Channel, inputType=FunctionInput_Vector2}
  FunctionOutput(None=ComponentMask(None=@n1) {r=true}) {description=Gradient along U axis of UV0, outputName=UGradient}
  FunctionOutput(None=ComponentMask(None=@n1) {g=true}) {description=Gradient along V axis of UV0, outputName=VGradient, sortPriority=1}
 ]
}
