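# Material function that generates a diamond-shaped gradient from UV channel 0, exposed to the library under Misc and Gradient.
# LinearGradient U and V outputs are each remapped into a tent ramp that peaks at the centre, via offset, scale, absolute value and one-minus.
# The two ramps are multiplied and raised to a Falloff exponent input, which defaults to 3, to produce the DiamondGradient output.
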MaterialFunction /Engine/Functions/Engine_MaterialFunctions01/Gradient/DiamondGradient
graph {
 properties {bExposeToLibrary=true, description=Uses UV Channel 0 to generate a Diamond Gradient, libraryCategoriesText [Misc,Gradient]}
 nodes [
  n1 = MaterialFunctionCall() {materialFunction=&LinearGradient}
  n2 = Constant() {r=-0.5}
  n3 = Add(A=@n1.UGradient, B=@n2)
  n4 = Constant() {r=2}
  n5 = OneMinus(None=Abs(None=Multiply(A=@n3, B=@n4)))
  n6 = Add(A=@n1.VGradient, B=@n2)
  n7 = OneMinus(None=Abs(None=Multiply(A=@n6, B=@n4)))
  n8 = FunctionInput() {
   bUsePreviewValueAsDefault=true
   description=Power Exponent for the Diamond
   inputName=Falloff
   inputType=FunctionInput_Scalar
   previewValue=<3 0 0 1>
  }
  FunctionOutput(None=Power(Base=Multiply(A=@n5, B=@n7), Exp=@n8)) {outputName=DiamondGradient}
 ]
}
