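# RadialGradient is a material function in the Misc and Gradient library categories that builds a radial gradient from UV channel 0.
# A SphereMask measures the texture coordinates against a CenterPosition input within a Radius input given as a fraction of UV space.
# Hardness is set to zero for a soft falloff, and both inputs use their preview values as defaults, giving a centered gradient when left unconnected.
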MaterialFunction /Engine/Functions/Engine_MaterialFunctions01/Gradient/RadialGradient
graph {
 properties {description=Uses UV Channel 0 to create a radial gradient, libraryCategoriesText [Misc,Gradient]}
 nodes [
  n1 = FunctionInput() {
   bUsePreviewValueAsDefault=true
   description=Position in UV space
   inputName=CenterPosition
   inputType=FunctionInput_Vector2
   previewValue=<0.5 0.5 0 1>
  }
  n2 = FunctionInput() {
   bUsePreviewValueAsDefault=true
   description=Radius in fraction of UV Space
   inputName=Radius
   inputType=FunctionInput_Scalar
   previewValue=<0.5 0 0 1>
  }
  FunctionOutput(None=SphereMask(A=TextureCoordinate(), B=@n1, Radius=@n2) {hardnessPercent=0}) {outputName=RadialGradient}
 ]
}
