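# Library material function that builds a radial gradient with exponential falloff from UVs, a center position, a radius and a density.
# UV-to-center distance is divided by Radius, then fed to ExponentialDensity with Density.
# The Invert Density static switch picks density on normalized distance when true, otherwise one minus density on inverted distance.
# Several nodes, including a clamp and a square-root chain, do not reach the output.
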MaterialFunction /Engine/Functions/Engine_MaterialFunctions01/Gradient/RadialGradientExponential
graph {
 properties {
  bExposeToLibrary=true
  description=Creates a radial gradient using the input UV coordinates and applies exponential density with specified density.
  libraryCategoriesText [Misc,Gradient]
 }
 nodes [
  n1 = FunctionInput(Preview=TextureCoordinate()) {bUsePreviewValueAsDefault=true, description=Position in UV space, inputName=UVs, inputType=FunctionInput_Vector2, previewValue=<0.5 0.5 0 1>}
  n2 = FunctionInput() {
   bUsePreviewValueAsDefault=true
   description=Position in UV space
   inputName=CenterPosition
   inputType=FunctionInput_Vector2
   sortPriority=1
   previewValue=<0.5 0.5 0 1>
  }
  n3 = Multiply(A=Distance(A=@n1, B=@n2))
  n4 = FunctionInput() {
   bUsePreviewValueAsDefault=true
   description=Radius in fraction of UV Space
   inputName=Radius
   inputType=FunctionInput_Scalar
   sortPriority=2
   previewValue=<0.5 0 0 1>
  }
  n5 = Divide(A=@n3, B=@n4)
  n6 = OneMinus(None=@n5)
  Clamp(None=@n6)
  Divide(A=@n1, B=@n4)
  Divide() {constB=1}
  n7 = FunctionInput() {bUsePreviewValueAsDefault=true, inputName=Density, inputType=FunctionInput_Scalar, sortPriority=3, previewValue=<2.333 0 0 1>}
  n8 = MaterialFunctionCall!(Depth=@n5, Density=@n7, Use Exp2) {materialFunction=/Engine/Functions/Engine_MaterialFunctions01/Density/ExponentialDensity}
  n9 = MaterialFunctionCall(Depth=@n6, Density=@n7) {materialFunction=&ExponentialDensity}
  n10 = StaticBool()
  FunctionOutput(None=StaticSwitch(
    True=@n8.Result
    False=OneMinus(None=@n9.Result)
    Value=FunctionInput(Preview=@n10) {bUsePreviewValueAsDefault=true, inputName=Invert Density, inputType=FunctionInput_StaticBool, sortPriority=4}
   )) {outputName=RadialGradientExponential}
  OneMinus(None=@n8.Result)
  SquareRoot(None=Subtract(A=Multiply(A=@n4, B=@n4) {desc="c^2"}, B=Multiply(A=@n3, B=@n3) {desc="a^2"}))
 ]
}
