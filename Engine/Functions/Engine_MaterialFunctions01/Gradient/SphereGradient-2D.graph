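# SphereGradient-2D is a library material function that generates a gradient representing the depth of a sphere centered on a chosen UV point.
# Distance from UVs to CenterPosition, clamped to Radius, feeds the Pythagorean square root of radius squared minus distance squared, giving half the sphere's thickness.
# It outputs that depth normalized to 0-1, in radius units, and doubled as full diameter thickness.
# Two trailing Multiply nodes connect to no output.
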MaterialFunction /Engine/Functions/Engine_MaterialFunctions01/Gradient/SphereGradient-2D
graph {
 properties {bExposeToLibrary=true, description="Generates a gradient representing the depth of a shere centered on the desired point. "}
 nodes [
  n1 = FunctionInput() {
   bUsePreviewValueAsDefault=true
   description=Radius in fraction of UV Space
   inputName=Radius
   inputType=FunctionInput_Scalar
   sortPriority=2
   previewValue=<0.5 0 0 1>
  }
  n2 = Multiply(A=@n1, B=@n1) {desc="c^2"}
  n3 = FunctionInput(Preview=TextureCoordinate()) {bUsePreviewValueAsDefault=true, description=Position in UV space, inputName=UVs, inputType=FunctionInput_Vector2, previewValue=<0.5 0.5 0 1>}
  n4 = FunctionInput() {
   bUsePreviewValueAsDefault=true
   description=Position in UV space
   inputName=CenterPosition
   inputType=FunctionInput_Vector2
   sortPriority=1
   previewValue=<0.5 0.5 0 1>
  }
  n5 = Clamp(None=Multiply(A=Distance(A=@n3, B=@n4)), Max=@n1)
  n6 = Multiply(A=@n5, B=@n5) {desc="a^2"}
  n7 = SquareRoot(None=Max(A=Subtract(A=@n2, B=@n6)) {constB=0})
  n8 = Clamp(None=Divide(A=@n7, B=@n1))
  FunctionOutput(None=@n8) {outputName=Result 0-1}
  FunctionOutput(None=Multiply(A=@n8, B=@n1)) {outputName=Result Radius, sortPriority=2}
  FunctionOutput(None=Multiply(A=@n7) {constB=2}) {outputName=Result Diameter, sortPriority=1}
  Multiply(A=@n1) {constB=2}
  Multiply(A=@n7, B=@n1)
 ]
}
