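# Material function that generates a gradient from the view ray's thickness through a virtual sphere defined by Location, Offset and Radius.
# Analytic ray-sphere intersection yields chord length, output as diameter, radius and 0-1 variants.
# Calculate Camera Inside replaces the cheap double half-chord with entry/exit distances, handling interior cameras and clipping exit at world depth behind translucency.
# Otherwise, optional Depth Biased Alpha fades results against SceneDepth over FadeDistance, using actor depth offset by sphere thickness.
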MaterialFunction /Engine/Functions/Engine_MaterialFunctions01/Gradient/SphereGradient-3D
graph {
 properties {bExposeToLibrary=true, description="Generates a gradient representing the depth of a shere centered on the desired point. "}
 nodes [
  n1 = Constant() {r=1}
  n2 = ActorPositionWS()
  n3 = CameraPositionWS()
  n4 = CameraVectorWS()
  n5 = MaterialFunctionCall() {materialFunction=&CameraDirectionVector}
  n6 = Multiply(A=Distance(A=Subtract(A=@n2, B=@n3), B=Constant()) {desc=ActorDepth}, B=Abs(None=DotProduct(A=@n4, B=@n5.Result)))
  n7 = CameraPositionWS() {desc=o}
  n8 = Add(
   A=FunctionInput(Preview=ActorPositionWS()) {bUsePreviewValueAsDefault=true, inputName=Location}
   B=FunctionInput() {bUsePreviewValueAsDefault=true, inputName=Offset, sortPriority=1}
  ) {desc=C}
  MaterialExpressionSubtract_39 = Subtract(A=@n7, B=@n8) {desc=L}
  n9 = Custom!(x=@MaterialExpressionSubtract_39) {
   code="length(x)"
   outputType=CMOT_Float1
   inputs [{inputName=x, input {expression=&self:MaterialExpressionSubtract_39}}]
  }
  n10 = CameraVectorWS()
  n11 = DotProduct(A=@MaterialExpressionSubtract_39, B=@n10) {desc=Tca}
  n12 = SquareRoot(None=Subtract(A=Multiply(A=@n9, B=@n9), B=Multiply(A=@n11, B=@n11))) {desc=d}
  n13 = FunctionInput() {bUsePreviewValueAsDefault=true, inputName=Radius, inputType=FunctionInput_Scalar, sortPriority=2, previewValue=<128 0 0 1>}
  n14 = Constant()
  n15 = Distance(A=@n8, B=@n7)
  n16 = SquareRoot(None=Subtract(A=Multiply(A=@n13, B=@n13) {desc="r^2"}, B=Multiply(A=@n12, B=@n12) {desc="d^2"})) {desc=Thc}
  n17 = Multiply(A=Add(A=@n11, B=@n16) {desc=T1}, B=CameraVectorWS()) {desc=T1r}
  n18 = CameraPositionWS() {desc=o}
  MaterialExpressionSubtract_6 = Subtract(
   A=CameraPositionWS()
   B=MaterialFunctionCall() {materialFunction=&WorldPositionBehindTranslucency}.Result
  )
  n19 = Custom(x=@MaterialExpressionSubtract_6) {
   code="length(x)"
   inputs [{inputName=x, input {expression=&self:MaterialExpressionSubtract_6}}]
  }
  n20 = Min(A=Distance(A=CameraPositionWS() {desc=o}, B=Add(A=@n17, B=@n18) {desc=Intersection2}) {desc=The point the camera intersects}, B=@n19)
  n21 = CameraPositionWS() {desc=o}
  n22 = Add(A=Multiply(A=Subtract(A=@n11, B=@n16), B=CameraVectorWS()), B=CameraPositionWS() {desc=o})
  n23 = Max(A=Subtract(A=@n20, B=Distance(A=@n21, B=@n22) {desc=The point the camera intersects}) {desc="Exit - Enter = Depth"}) {constB=0}
  n24 = Constant()
  n25 = Multiply(A=@n16) {constB=2, desc=Cheap method}
  n26 = Constant()
  n27 = FunctionInput(Preview=StaticBool()) {
   bUsePreviewValueAsDefault=true
   description=Extra calculations need to be done to calculate sphere intersection when the camera is inside and use real world depths for intersection.
   inputName=Calculate Camera Inside
   inputType=FunctionInput_StaticBool
   sortPriority=4
  }
  n28 = If(
   A=@n12
   B=@n13
   "A > B"=@n14
   "A == B"=@n14
   "A < B"=StaticSwitch(
    True=If(A=@n15, B=@n13, "A > B"=@n23, "A == B"=@n23, "A < B"=@n20) {desc=Is camera inside Sphere?}
    False=If(A=@n11, B=@n24, "A > B"=@n25, "A == B"=@n25, "A < B"=@n26)
    Value=@n27
   )
  )
  n29 = Divide(A=@n28)
  n30 = Multiply(A=@n29)
  n31 = FunctionInput() {
   bUsePreviewValueAsDefault=true
   description="Extra calculations need to be done to calculate sphere intersection when the camera is inside, so enable only when needed."
   inputName=FadeDistance
   inputType=FunctionInput_Scalar
   sortPriority=6
   previewValue=<1 0 0 1>
  }
  n32 = Divide(A=Subtract(A=SceneDepth().R, B=Subtract(A=@n6, B=@n30) {constB=0, desc=Add Sphere}), B=@n31)
  n33 = FunctionInput(Preview=StaticBool()) {
   bUsePreviewValueAsDefault=true
   description="Extra calculations need to be done to calculate sphere intersection when the camera is inside, so enable only when needed."
   inputName=Depth Biased Alpha
   inputType=FunctionInput_StaticBool
   sortPriority=5
  }
  n34 = StaticSwitch(True=StaticSwitch(True=@n1, False=Clamp(None=@n32), Value=@n27), False=@n1, Value=@n33)
  FunctionOutput(None=Multiply(A=@n34, B=@n28)) {outputName=Result Diameter, sortPriority=1}
  n35 = Multiply(A=@n13) {constB=2}
  FunctionOutput(None=Multiply(A=@n34, B=Divide(A=@n28, B=@n35))) {outputName=Result 0-1}
  FunctionOutput(None=Multiply(A=@n34, B=@n29)) {outputName=Result Radius, sortPriority=2}
  Normalize(VectorInput=@n10)
 ]
}
